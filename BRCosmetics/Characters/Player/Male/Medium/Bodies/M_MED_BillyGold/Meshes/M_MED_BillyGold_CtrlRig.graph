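# Control Rig for the BillyGold medium male body that procedurally drives corrective and accessory bones.
# Forward solve chains twist fix, shared deform rig, knee, jacket, belt and cartridge functions, using weighted parent constraints and additive offsets.
# DeltaRotation turns swing or twist against mostly construction-added nulls into corrective weights.
# Jacket caches thigh-bend weights that belts reuse, alongside spine bend and a spherical pose reader; FK control functions are unwired.
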ControlRigBlueprint /BRCosmetics/Characters/Player/Male/Medium/Bodies/M_MED_BillyGold/Meshes/M_MED_BillyGold_CtrlRig
graphs [
 CollapseNode_ContainedGraph {nodes [
   n1 = RigVMFunctionEntryNode()
   &RigUnit_ParentConstraint(
    AdvancedSettings=(RotationOrderForFilter=XZY)
    Child=(Type=Bone,Name="base_hemJacket_1_l")
    Parents="((Item=(Type=Bone,Name=\"spine_05\"),Weight=0.0011),(Item=(Type=Bone,Name=\"spine_04\"),Weight=0.0776),(Item=(Type=Bone,Name=\"spine_03\"),Weight=0.1246),(Item=(Type=Bone,Name=\"pelvis\"),Weight=0.1444),(Item=(Type=Bone,Name=\"spine_01\"),Weight=0.196),(Item=(Type=B ...(truncated)"
    Weight=1
    ExecutePin=@n1.ExecutePin
   )
   > &RigUnit_ParentConstraint(
    AdvancedSettings=(RotationOrderForFilter=XZY)
    Child=(Type=Bone,Name="base_hemJacket_1_r")
    Parents="((Item=(Type=Bone,Name=\"spine_05\"),Weight=0.0011),(Item=(Type=Bone,Name=\"spine_04\"),Weight=0.0776),(Item=(Type=Bone,Name=\"spine_03\"),Weight=0.1246),(Item=(Type=Bone,Name=\"pelvis\"),Weight=0.1444),(Item=(Type=Bone,Name=\"spine_01\"),Weight=0.196),(Item=(Type=B ...(truncated)"
    Weight=1
   )
   > &RigUnit_ParentConstraint(
    AdvancedSettings=(RotationOrderForFilter=XZY)
    Child=(Type=Bone,Name="base_hemJacket_2_l")
    Parents="((Item=(Type=Bone,Name=\"spine_05\"),Weight=0.001),(Item=(Type=Bone,Name=\"spine_04\"),Weight=0.0442),(Item=(Type=Bone,Name=\"spine_03\"),Weight=0.0729),(Item=(Type=Bone,Name=\"pelvis\"),Weight=0.2111),(Item=(Type=Bone,Name=\"spine_01\"),Weight=0.2578),(Item=(Type=B ...(truncated)"
    Weight=1
   )
   > &RigUnit_ParentConstraint(
    AdvancedSettings=(RotationOrderForFilter=XZY)
    Child=(Type=Bone,Name="base_hemJacket_3_l")
    Parents="((Item=(Type=Bone,Name=\"spine_05\"),Weight=0.0021),(Item=(Type=Bone,Name=\"spine_04\"),Weight=0.0509),(Item=(Type=Bone,Name=\"spine_03\"),Weight=0.0711),(Item=(Type=Bone,Name=\"pelvis\"),Weight=0.1901),(Item=(Type=Bone,Name=\"spine_01\"),Weight=0.2869),(Item=(Type= ...(truncated)"
    Weight=1
   )
   > &RigUnit_ParentConstraint(
    AdvancedSettings=(RotationOrderForFilter=XZY)
    Child=(Type=Bone,Name="base_hemJacket_1_c")
    Parents="((Item=(Type=Bone,Name=\"spine_05\"),Weight=0.0031),(Item=(Type=Bone,Name=\"spine_04\"),Weight=0.0544),(Item=(Type=Bone,Name=\"spine_03\"),Weight=0.0748),(Item=(Type=Bone,Name=\"pelvis\"),Weight=0.1822),(Item=(Type=Bone,Name=\"spine_01\"),Weight=0.3063),(Item=(Type= ...(truncated)"
    Weight=1
   )
   > &RigUnit_ParentConstraint(
    AdvancedSettings=(RotationOrderForFilter=XZY)
    Child=(Type=Bone,Name="base_hemJacket_2_r")
    Parents="((Item=(Type=Bone,Name=\"spine_05\"),Weight=0.001),(Item=(Type=Bone,Name=\"spine_04\"),Weight=0.0442),(Item=(Type=Bone,Name=\"spine_03\"),Weight=0.0729),(Item=(Type=Bone,Name=\"pelvis\"),Weight=0.2111),(Item=(Type=Bone,Name=\"spine_01\"),Weight=0.2578),(Item=(Type=B ...(truncated)"
    Weight=1
   )
   > n2 = &RigUnit_ParentConstraint(
    AdvancedSettings=(RotationOrderForFilter=XZY)
    Child=(Type=Bone,Name="base_hemJacket_3_r")
    Parents="((Item=(Type=Bone,Name=\"spine_05\"),Weight=0.0021),(Item=(Type=Bone,Name=\"spine_04\"),Weight=0.0509),(Item=(Type=Bone,Name=\"spine_03\"),Weight=0.0711),(Item=(Type=Bone,Name=\"pelvis\"),Weight=0.1901),(Item=(Type=Bone,Name=\"spine_01\"),Weight=0.2869),(Item=(Type= ...(truncated)"
    Weight=1
   )
   RigVMFunctionReturnNode(ExecutePin=@n2.ExecutePin)
   &RigUnit_ParentConstraint(
    AdvancedSettings=(RotationOrderForFilter=XZY)
    Child=(Type=Bone,Name="base_hemJacket_1_l")
    ExecutePin=()
    Parents=((Item=(Type=Bone,Name="spine_04"),Weight=0.0174),(Item=(Type=Bone,Name="spine_03"),Weight=0.1934),(Item=(Type=Bone,Name="spine_01"),Weight=0.2261),(Item=(Type=Bone,Name="pelvis"),Weight=0.26),(Item=(Type=Bone,Name="spine_02"),Weight=0.303))
    Weight=1
    bMaintainOffset=true
   )
   > &RigUnit_ParentConstraint(
    AdvancedSettings=(RotationOrderForFilter=XZY)
    Child=(Type=Bone,Name="base_hemJacket_1_r")
    Parents=((Item=(Type=Bone,Name="spine_04"),Weight=0.0174),(Item=(Type=Bone,Name="spine_03"),Weight=0.1934),(Item=(Type=Bone,Name="spine_01"),Weight=0.2261),(Item=(Type=Bone,Name="pelvis"),Weight=0.26),(Item=(Type=Bone,Name="spine_02"),Weight=0.303))
    Weight=1
   )
   > &RigUnit_ParentConstraint(
    AdvancedSettings=(RotationOrderForFilter=XZY)
    Child=(Type=Bone,Name="base_hemJacket_2_l")
    Parents=((Item=(Type=Bone,Name="spine_03"),Weight=0.1307),(Item=(Type=Bone,Name="spine_01"),Weight=0.1851),(Item=(Type=Bone,Name="spine_02"),Weight=0.2303),(Item=(Type=Bone,Name="pelvis"),Weight=0.4539))
    Weight=1
   )
   > &RigUnit_ParentConstraint(
    AdvancedSettings=(RotationOrderForFilter=XZY)
    Child=(Type=Bone,Name="base_hemJacket_3_l")
    Parents=((Item=(Type=Bone,Name="spine_04"),Weight=0.0135),(Item=(Type=Bone,Name="spine_03"),Weight=0.1408),(Item=(Type=Bone,Name="spine_01"),Weight=0.1898),(Item=(Type=Bone,Name="spine_02"),Weight=0.237),(Item=(Type=Bone,Name="pelvis"),Weight=0.4189))
    Weight=1
   )
   > &RigUnit_ParentConstraint(
    AdvancedSettings=(RotationOrderForFilter=XZY)
    Child=(Type=Bone,Name="base_hemJacket_1_c")
    Parents=((Item=(Type=Bone,Name="spine_04"),Weight=0.0254),(Item=(Type=Bone,Name="spine_03"),Weight=0.1658),(Item=(Type=Bone,Name="spine_01"),Weight=0.1929),(Item=(Type=Bone,Name="spine_02"),Weight=0.2137),(Item=(Type=Bone,Name="pelvis"),Weight=0.4022))
    Weight=1
   )
   > &RigUnit_ParentConstraint(
    AdvancedSettings=(RotationOrderForFilter=XZY)
    Child=(Type=Bone,Name="base_hemJacket_2_r")
    Parents=((Item=(Type=Bone,Name="spine_03"),Weight=0.1307),(Item=(Type=Bone,Name="spine_01"),Weight=0.1851),(Item=(Type=Bone,Name="spine_02"),Weight=0.2303),(Item=(Type=Bone,Name="pelvis"),Weight=0.4539))
    Weight=1
   )
   > &RigUnit_ParentConstraint(
    AdvancedSettings=(RotationOrderForFilter=XZY)
    Child=(Type=Bone,Name="base_hemJacket_3_r")
    Parents=((Item=(Type=Bone,Name="spine_04"),Weight=0.0135),(Item=(Type=Bone,Name="spine_03"),Weight=0.1408),(Item=(Type=Bone,Name="spine_01"),Weight=0.1898),(Item=(Type=Bone,Name="spine_02"),Weight=0.237),(Item=(Type=Bone,Name="pelvis"),Weight=0.4189))
    Weight=1
   )
  ]}
 CollapseNode_ContainedGraph {nodes [
   n1 = RigVMFunctionEntryNode()
   &RigUnit_ParentConstraint(
    AdvancedSettings=(RotationOrderForFilter=XZY)
    Child=(Type=Bone,Name="kin_belt_B_1")
    Parents="((Item=(Type=Bone,Name=\"spine_03\"),Weight=0.0004),(Item=(Type=Bone,Name=\"thigh_l\"),Weight=0.0095),(Item=(Type=Bone,Name=\"spine_02\"),Weight=0.0149),(Item=(Type=Bone,Name=\"spine_01\"),Weight=0.0483),(Item=(Type=Bone,Name=\"deform_groin_l\"),Weight=0.0543),(Item ...(truncated)"
    Weight=1
    ExecutePin=@n1.ExecutePin
   )
   > &RigUnit_ParentConstraint(
    AdvancedSettings=(RotationOrderForFilter=XZY)
    Child=(Type=Bone,Name="kin_belt_B_2")
    Parents="((Item=(Type=Bone,Name=\"thigh_l\"),Weight=0.0009),(Item=(Type=Bone,Name=\"spine_03\"),Weight=0.002),(Item=(Type=Bone,Name=\"thigh_twist_01_r\"),Weight=0.0046),(Item=(Type=Bone,Name=\"spine_02\"),Weight=0.0466),(Item=(Type=Bone,Name=\"spine_01\"),Weight=0.0885),(Ite ...(truncated)"
    Weight=1
   )
   > &RigUnit_ParentConstraint(
    AdvancedSettings=(RotationOrderForFilter=XZY)
    Child=(Type=Bone,Name="kin_belt_B_3")
    Parents="((Item=(Type=Bone,Name=\"deform_groin_r\"),Weight=0.0038),(Item=(Type=Bone,Name=\"deform_groin_l\"),Weight=0.0091),(Item=(Type=Bone,Name=\"spine_03\"),Weight=0.0139),(Item=(Type=Bone,Name=\"thigh_twist_01_r\"),Weight=0.0171),(Item=(Type=Bone,Name=\"thigh_twist_01_l ...(truncated)"
    Weight=1
   )
   > &RigUnit_ParentConstraint(
    AdvancedSettings=(RotationOrderForFilter=XZY)
    Child=(Type=Bone,Name="kin_belt_B_4")
    Parents="((Item=(Type=Bone,Name=\"thigh_twist_01_l\"),Weight=0.003),(Item=(Type=Bone,Name=\"spine_04\"),Weight=0.0051),(Item=(Type=Bone,Name=\"deform_groin_r\"),Weight=0.0139),(Item=(Type=Bone,Name=\"thigh_twist_01_r\"),Weight=0.0215),(Item=(Type=Bone,Name=\"spine_03\"),Weig ...(truncated)"
    Weight=1
   )
   > &RigUnit_ParentConstraint(
    AdvancedSettings=(RotationOrderForFilter=XZY)
    Child=(Type=Bone,Name="kin_belt_B_5")
    Parents="((Item=(Type=Bone,Name=\"spine_04\"),Weight=0.0056),(Item=(Type=Bone,Name=\"deform_groin_r\"),Weight=0.0099),(Item=(Type=Bone,Name=\"spine_03\"),Weight=0.0159),(Item=(Type=Bone,Name=\"thigh_twist_01_r\"),Weight=0.0286),(Item=(Type=Bone,Name=\"spine_02\"),Weight=0.19 ...(truncated)"
    Weight=1
   )
   > &RigUnit_ParentConstraint(
    AdvancedSettings=(RotationOrderForFilter=XZY)
    Child=(Type=Bone,Name="kin_belt_B_6")
    Parents="((Item=(Type=Bone,Name=\"thigh_twist_01_l\"),Weight=0.0002),(Item=(Type=Bone,Name=\"spine_04\"),Weight=0.0039),(Item=(Type=Bone,Name=\"thigh_twist_01_r\"),Weight=0.0042),(Item=(Type=Bone,Name=\"deform_glute_r\"),Weight=0.009),(Item=(Type=Bone,Name=\"spine_03\"),Weig ...(truncated)"
    Weight=1
   )
   > &RigUnit_ParentConstraint(
    AdvancedSettings=(RotationOrderForFilter=XZY)
    Child=(Type=Bone,Name="kin_belt_B_7")
    Parents="((Item=(Type=Bone,Name=\"deform_glute_l\"),Weight=0.0021),(Item=(Type=Bone,Name=\"deform_glute_r\"),Weight=0.0032),(Item=(Type=Bone,Name=\"spine_03\"),Weight=0.0038),(Item=(Type=Bone,Name=\"thigh_twist_01_l\"),Weight=0.0209),(Item=(Type=Bone,Name=\"thigh_twist_01_r ...(truncated)"
    Weight=1
   )
   > &RigUnit_ParentConstraint(
    AdvancedSettings=(RotationOrderForFilter=XZY)
    Child=(Type=Bone,Name="kin_belt_B_8")
    Parents="((Item=(Type=Bone,Name=\"thigh_l\"),Weight=0.0009),(Item=(Type=Bone,Name=\"spine_03\"),Weight=0.0013),(Item=(Type=Bone,Name=\"thigh_twist_01_r\"),Weight=0.0014),(Item=(Type=Bone,Name=\"spine_02\"),Weight=0.0334),(Item=(Type=Bone,Name=\"thigh_twist_01_l\"),Weight=0.1 ...(truncated)"
    Weight=1
   )
   > &RigUnit_ParentConstraint(
    AdvancedSettings=(RotationOrderForFilter=XZY)
    Child=(Type=Bone,Name="kin_belt_A_1")
    Parents="((Item=(Type=Bone,Name=\"spine_03\"),Weight=0.0006),(Item=(Type=Bone,Name=\"thigh_r\"),Weight=0.0068),(Item=(Type=Bone,Name=\"spine_02\"),Weight=0.0198),(Item=(Type=Bone,Name=\"deform_groin_r\"),Weight=0.0513),(Item=(Type=Bone,Name=\"spine_01\"),Weight=0.061),(Item= ...(truncated)"
    Weight=1
   )
   > &RigUnit_ParentConstraint(
    AdvancedSettings=(RotationOrderForFilter=XZY)
    Child=(Type=Bone,Name="kin_belt_A_2")
    Parents="((Item=(Type=Bone,Name=\"spine_03\"),Weight=0.0009),(Item=(Type=Bone,Name=\"thigh_r\"),Weight=0.0025),(Item=(Type=Bone,Name=\"thigh_twist_01_l\"),Weight=0.0058),(Item=(Type=Bone,Name=\"spine_02\"),Weight=0.0242),(Item=(Type=Bone,Name=\"spine_01\"),Weight=0.0544),(It ...(truncated)"
    Weight=1
   )
   > &RigUnit_ParentConstraint(
    AdvancedSettings=(RotationOrderForFilter=XZY)
    Child=(Type=Bone,Name="kin_belt_A_3")
    Parents="((Item=(Type=Bone,Name=\"spine_03\"),Weight=0.007),(Item=(Type=Bone,Name=\"deform_groin_l\"),Weight=0.0172),(Item=(Type=Bone,Name=\"deform_groin_r\"),Weight=0.021),(Item=(Type=Bone,Name=\"thigh_twist_01_l\"),Weight=0.0419),(Item=(Type=Bone,Name=\"thigh_twist_01_r\") ...(truncated)"
    Weight=1
   )
   > &RigUnit_ParentConstraint(
    AdvancedSettings=(RotationOrderForFilter=XZY)
    Child=(Type=Bone,Name="kin_belt_A_4")
    Parents="((Item=(Type=Bone,Name=\"spine_04\"),Weight=0.0026),(Item=(Type=Bone,Name=\"thigh_twist_01_r\"),Weight=0.0066),(Item=(Type=Bone,Name=\"spine_03\"),Weight=0.013),(Item=(Type=Bone,Name=\"deform_groin_l\"),Weight=0.0323),(Item=(Type=Bone,Name=\"thigh_twist_01_l\"),Weig ...(truncated)"
    Weight=1
   )
   > &RigUnit_ParentConstraint(
    AdvancedSettings=(RotationOrderForFilter=XZY)
    Child=(Type=Bone,Name="kin_belt_A_5")
    Parents="((Item=(Type=Bone,Name=\"spine_04\"),Weight=0.0056),(Item=(Type=Bone,Name=\"deform_groin_l\"),Weight=0.0099),(Item=(Type=Bone,Name=\"spine_03\"),Weight=0.016),(Item=(Type=Bone,Name=\"thigh_twist_01_l\"),Weight=0.0281),(Item=(Type=Bone,Name=\"spine_02\"),Weight=0.199 ...(truncated)"
    Weight=1
   )
   > &RigUnit_ParentConstraint(
    AdvancedSettings=(RotationOrderForFilter=XZY)
    Child=(Type=Bone,Name="kin_belt_A_6")
    Parents="((Item=(Type=Bone,Name=\"thigh_twist_01_r\"),Weight=0.0002),(Item=(Type=Bone,Name=\"spine_04\"),Weight=0.0038),(Item=(Type=Bone,Name=\"thigh_twist_01_l\"),Weight=0.0042),(Item=(Type=Bone,Name=\"deform_glute_l\"),Weight=0.0091),(Item=(Type=Bone,Name=\"spine_03\"),Wei ...(truncated)"
    Weight=1
   )
   > &RigUnit_ParentConstraint(
    AdvancedSettings=(RotationOrderForFilter=XZY)
    Child=(Type=Bone,Name="kin_belt_A_7")
    Parents="((Item=(Type=Bone,Name=\"deform_glute_l\"),Weight=0.0021),(Item=(Type=Bone,Name=\"deform_glute_r\"),Weight=0.0032),(Item=(Type=Bone,Name=\"spine_03\"),Weight=0.0038),(Item=(Type=Bone,Name=\"thigh_twist_01_l\"),Weight=0.0209),(Item=(Type=Bone,Name=\"thigh_twist_01_r ...(truncated)"
    Weight=1
   )
   > n2 = &RigUnit_ParentConstraint(
    AdvancedSettings=(RotationOrderForFilter=XZY)
    Child=(Type=Bone,Name="kin_belt_A_8")
    Parents="((Item=(Type=Bone,Name=\"thigh_r\"),Weight=0.0009),(Item=(Type=Bone,Name=\"spine_03\"),Weight=0.0013),(Item=(Type=Bone,Name=\"thigh_twist_01_l\"),Weight=0.0014),(Item=(Type=Bone,Name=\"spine_02\"),Weight=0.0336),(Item=(Type=Bone,Name=\"thigh_twist_01_r\"),Weight=0.1 ...(truncated)"
    Weight=1
   )
   RigVMFunctionReturnNode(ExecutePin=@n2.ExecutePin)
  ]}
 New Function_ContainedGraph {nodes [
   n1 = RigVMFunctionEntryNode()
   RigVMFunction_MathBoolToInteger = &RigVMFunction_MathBoolToInteger(Value=@n1.onlyTwist)
   RigVMFunction_MathQuaternionSwingTwist = &RigVMFunction_MathQuaternionSwingTwist(
    TwistAxis=@n1.twistAxis
    Input=&RigUnit_GetRelativeTransformForItem(bChildInitial=False, bParentInitial=False, Child=@n1.child, Parent=@n1.parent).RelativeTransform.Rotation
   )
   n2 = RigVMDispatchNode(
    Values=((),())
    Index=@RigVMFunction_MathBoolToInteger.Result
    Values.0=@RigVMFunction_MathQuaternionSwingTwist.Swing
    Values.1=@RigVMFunction_MathQuaternionSwingTwist.Twist
   ) {template="DISPATCH_RigVMDispatch_SelectInt32(in Index,in Values,out Result)"}
   RigVMFunction_MathQuaternionToEuler = &RigVMFunction_MathQuaternionToEuler(Value=@n2.Result, RotationOrder=@n1.Rotation Order)
   > DISPATCH_RigVMDispatch_Print = RigVMDispatchNode(ScreenColor=(R=1,G=1,B=1,A=1), ScreenDuration=1.5, Enabled=@n1.debugRotation, Value=@RigVMFunction_MathQuaternionToEuler.Result) {template="DISPATCH_RigVMDispatch_Print(in Prefix,in Value,in Enabled,in ScreenDuration,in ScreenColor)"}
   Cast = RigVMDispatchNode(Value=@n1.triggerAxis) {template="DISPATCH_RigVMDispatch_CastEnumToInt(in Value,out Result)"}
   DISPATCH_RigVMDispatch_SelectInt32 = RigVMDispatchNode(
    Values=(0,0,0)
    Index=@Cast.Result
    Values.0=@RigVMFunction_MathQuaternionToEuler.Result.X
    Values.1=@RigVMFunction_MathQuaternionToEuler.Result.Y
    Values.2=@RigVMFunction_MathQuaternionToEuler.Result.Z
   ) {template="DISPATCH_RigVMDispatch_SelectInt32(in Index,in Values,out Result)"}
   Remap = &RigVMFunction_MathDoubleRemap(
    bClamp=False
    Value=@DISPATCH_RigVMDispatch_SelectInt32.Result
    TargetMinimum=@n1.outputMinMax.X
    TargetMaximum=@n1.outputMinMax.Y
    SourceMinimum=@n1.rotationMinMax.X
    SourceMaximum=@n1.rotationMinMax.Y
   )
   Absolute = &RigVMFunction_MathDoubleAbs(Value=@Remap.Result)
   Multiply = &RigVMFunction_MathDoubleMul(
    A=&RigVMFunction_MathDoubleClamp(Value=RigVMDispatchNode(True=@Absolute.Result, Condition=@n1.absoluteValue, False=@Remap.Result) {template="DISPATCH_RigVMDispatch_If(in Condition,in True,in False,out Result)"}.Result).Result
    B=@n1.premult
   )
   > RigVMFunctionReturnNode(outputQuat=@n2.Result, outputWeight=@Multiply.Result)
  ]}
 New Function_ContainedGraph {nodes [
   RigVMFunctionEntryNode()
   > DeltaRotation v4 = RigVMFunctionReferenceNode(
    Rotation Order=ZYX
    child=(Type=Bone,Name="hand_l")
    onlyTwist=true
    outputMinMax=(X=0,Y=1)
    parent=(Type=Null,Name="hand_l_null")
    premult=1
    rotationMinMax=(X=70,Y=100)
    twistAxis=(X=1,Y=0,Z=0)
   )
   > &RigUnit_ModifyTransforms(
    ItemToModify="((Item=(Type=Bone,Name=\"lowerarm_twist_01_l\"),Transform=(Rotation=(X=0,Y=-0.0262,Z=0,W=0.9997),Translation=(X=-0.7,Y=-0.17,Z=1),Scale3D=(X=1,Y=1,Z=1))),(Item=(Type=Bone,Name=\"lowerarm_twist_02_l\"),Transform=(Rotation=(X=0.342,Y=0,Z=0,W=0.9397),Translation= ...(truncated)"
    WeightMinimum=0
    Weight=@"DeltaRotation v4".outputWeight
   )
   > RigVMFunctionReturnNode()
  ]}
 New Function_ContainedGraph {nodes [
   RigVMFunctionEntryNode()
   > RigVMFunction_Sequence = RigVMAggregateNode()
   constraints = RigVMCollapseNode(ExecutePin=@RigVMFunction_Sequence.A)
   DeltaRotation v4 = RigVMFunctionReferenceNode(
    Rotation Order=ZYX
    child=(Type=Bone,Name="thigh_l")
    debugRotation=false
    outputMinMax=(X=0,Y=1)
    parent=(Type=Null,Name="thigh_l_null")
    premult=1
    rotationMinMax=(X=-20,Y=-110)
    triggerAxis=Z
    twistAxis=(X=1,Y=0,Z=0)
    ExecuteContext=@constraints.ExecutePin
   )
   > RigVMVariableNode(Value=@"DeltaRotation v4".outputWeight) {variable=thigh_l_n20_n110}
   > &RigUnit_ModifyTransforms(
    ItemToModify=((Item=(Type=Bone,Name="base_hemJacket_1_l"),Transform=(Rotation=(X=0.154,Y=-0.0802,Z=-0.4547,W=0.8735),Translation=(X=0,Y=0,Z=-1),Scale3D=(X=1,Y=1,Z=1))))
    WeightMinimum=0
    Weight=@"DeltaRotation v4".outputWeight
   )
   > DeltaRotation v4_1 = RigVMFunctionReferenceNode(
    Rotation Order=ZYX
    child=(Type=Bone,Name="thigh_r")
    debugRotation=false
    outputMinMax=(X=0,Y=1)
    parent=(Type=Null,Name="thigh_r_null")
    premult=1
    rotationMinMax=(X=-20,Y=-110)
    triggerAxis=Z
    twistAxis=(X=1,Y=0,Z=0)
   )
   > RigVMVariableNode(Value=@"DeltaRotation v4_1".outputWeight) {variable=thigh_r_n20_n110}
   > &RigUnit_ModifyTransforms(
    ItemToModify=((Item=(Type=Bone,Name="base_hemJacket_1_r"),Transform=(Rotation=(X=-0.154,Y=0.0802,Z=-0.4547,W=0.8735),Translation=(X=0,Y=0,Z=-1),Scale3D=(X=1,Y=1,Z=1))))
    WeightMinimum=0
    Weight=@"DeltaRotation v4_1".outputWeight
   )
   DeltaRotation v4_2 = RigVMFunctionReferenceNode(
    Rotation Order=ZYX
    child=(Type=Bone,Name="clavicle_l")
    outputMinMax=(X=0,Y=1)
    parent=(Type=Null,Name="clavicle_l_null")
    premult=0.5
    rotationMinMax=(X=0,Y=-40)
    triggerAxis=Y
    twistAxis=(X=1,Y=0,Z=0)
    ExecuteContext=@RigVMFunction_Sequence.B
   )
   > &RigUnit_ModifyTransforms(
    ItemToModify=((Item=(Type=Bone,Name="deform_lat_l"),Transform=(Rotation=(X=0,Y=0,Z=0,W=1),Translation=(X=2,Y=0,Z=8),Scale3D=(X=1,Y=1,Z=1))))
    WeightMinimum=0
    Weight=@"DeltaRotation v4_2".outputWeight
   )
   > DeltaRotation v4_3 = RigVMFunctionReferenceNode(
    Rotation Order=ZYX
    child=(Type=Bone,Name="clavicle_r")
    outputMinMax=(X=0,Y=1)
    parent=(Type=Null,Name="clavicle_r_null")
    premult=0.5
    rotationMinMax=(X=0,Y=-40)
    triggerAxis=Y
    twistAxis=(X=1,Y=0,Z=0)
   )
   > &RigUnit_ModifyTransforms(
    ItemToModify=((Item=(Type=Bone,Name="deform_lat_r"),Transform=(Rotation=(X=0,Y=0,Z=0,W=1),Translation=(X=2,Y=0,Z=-8),Scale3D=(X=1,Y=1,Z=1))))
    WeightMinimum=0
    Weight=@"DeltaRotation v4_3".outputWeight
   )
   RigVMFunctionReturnNode(ExecuteContext=@RigVMFunction_Sequence.C)
  ]}
 New Function_ContainedGraph {nodes [
   n1 = RigVMFunctionEntryNode()
   constraints = RigVMCollapseNode(ExecutePin=@n1.ExecuteContext)
   &RigUnit_ModifyTransforms(
    ItemToModify="((Item=(Type=Bone,Name=\"kin_belt_B_2\"),Transform=(Rotation=(X=0,Y=0,Z=0,W=1),Translation=(X=4.2079,Y=5.1888,Z=-0.7756),Scale3D=(X=1,Y=1,Z=1))),(Item=(Type=Bone,Name=\"kin_belt_B_1\"),Transform=(Rotation=(X=0.0819,Y=0,Z=0,W=0.9966),Translation=(X=4.8117,Y=3.3 ...(truncated)"
    WeightMinimum=0
    Weight=RigVMVariableNode() {variable=thigh_l_n20_n110}.Value
    ExecutePin=@constraints.ExecutePin
   )
   n2 = RigVMVariableNode() {variable=thigh_r_n20_n110}
   > &RigUnit_ModifyTransforms(
    ItemToModify="((Item=(Type=Bone,Name=\"kin_belt_A_2\"),Transform=(Rotation=(X=0,Y=0,Z=0,W=1),Translation=(X=2.5488,Y=4.6353,Z=0.9968),Scale3D=(X=1,Y=1,Z=1))),(Item=(Type=Bone,Name=\"kin_belt_A_1\"),Transform=(Rotation=(X=-0.0259,Y=0.0042,Z=0.1614,W=0.9865),Translation=(X=3. ...(truncated)"
    WeightMinimum=0
    Weight=@n2.Value
   )
   > DeltaRotation v4 = RigVMFunctionReferenceNode(
    Rotation Order=ZYX
    child=(Type=Bone,Name="spine_02")
    debugRotation=false
    outputMinMax=(X=0,Y=1)
    parent=(Type=Bone,Name="pelvis")
    premult=1
    rotationMinMax=(X=10,Y=40)
    triggerAxis=Z
    twistAxis=(X=1,Y=0,Z=0)
   )
   > &RigUnit_ModifyTransforms(
    ItemToModify="((Item=(Type=Bone,Name=\"kin_belt_A_3\"),Transform=(Rotation=(X=0,Y=0,Z=0,W=1),Translation=(X=0,Y=2,Z=0),Scale3D=(X=1,Y=1,Z=1))),(Item=(Type=Bone,Name=\"kin_belt_B_3\"),Transform=(Rotation=(X=0,Y=0,Z=0,W=1),Translation=(X=0,Y=2,Z=0),Scale3D=(X=1,Y=1,Z=1))),(It ...(truncated)"
    WeightMinimum=0
    Weight=@"DeltaRotation v4".outputWeight
   )
   Multiply = &RigVMFunction_MathDoubleMul(A=@"DeltaRotation v4".outputWeight, B=@n2.Value)
   Clamp = &RigVMFunction_MathDoubleClamp(Value=&RigVMFunction_MathDoubleMul(B=2, A=@Multiply.Result).Result)
   > &RigUnit_ModifyTransforms(
    ItemToModify="((Item=(Type=Bone,Name=\"kin_belt_B_4\"),Transform=(Rotation=(X=0,Y=0,Z=0,W=1),Translation=(X=-0.1799,Y=2.6088,Z=-0.243),Scale3D=(X=1,Y=1,Z=1))),(Item=(Type=Bone,Name=\"kin_belt_B_5\"),Transform=(Rotation=(X=0,Y=0,Z=0,W=1),Translation=(X=2,Y=2,Z=0),Scale3D=(X= ...(truncated)"
    WeightMinimum=0
    Weight=@Clamp.Result
   )
   > n3 = &RigUnit_SphericalPoseReader(
    ActiveRegionScaleFactors=(PositiveWidth=1,NegativeWidth=1,PositiveHeight=1.5,NegativeHeight=1.5)
    ActiveRegionSize=0.06
    Debug=(bDrawDebug=false,bDraw2D=False,bDrawLocalAxes=False,DebugScale=25,DebugSegments=20,DebugThickness=0.25)
    DriverAxis=(X=1,Y=0,Z=0)
    DriverItem=(Type=Bone,Name="thigh_r")
    FalloffRegionScaleFactors=(PositiveWidth=1,NegativeWidth=1,PositiveHeight=1.5,NegativeHeight=1.5)
    FalloffSize=0.2
    FlipHeightScaling=False
    FlipWidthScaling=False
    OptionalParentItem=(Type=Bone,Name="None")
    RotationOffset=(X=-20,Y=-32,Z=0)
   )
   > &RigUnit_ModifyTransforms(
    ItemToModify=((Item=(Type=Bone,Name="kin_belt_A_1"),Transform=(Rotation=(X=0,Y=0.2588,Z=0,W=0.9659),Translation=(X=0,Y=1,Z=1),Scale3D=(X=1,Y=1,Z=1))))
    WeightMinimum=0
    Weight=@n3.OutputParam
   )
   > RigVMFunctionReturnNode()
  ]}
 New Function_ContainedGraph {nodes [
   RigVMFunctionEntryNode()
   > &RigUnit_ParentConstraint(
    AdvancedSettings=(RotationOrderForFilter=XZY)
    Child=(Type=Bone,Name="base_cartridges_l")
    Parents="((Item=(Type=Bone,Name=\"upperarm_twist_01_l\"),Weight=0.0001),(Item=(Type=Bone,Name=\"spine_02\"),Weight=0.0033),(Item=(Type=Bone,Name=\"spine_03\"),Weight=0.0292),(Item=(Type=Bone,Name=\"clavicle_l\"),Weight=0.0796),(Item=(Type=Bone,Name=\"spine_05\"),Weight=0.083 ...(truncated)"
    Weight=1
   )
   > &RigUnit_ParentConstraint(
    AdvancedSettings=(RotationOrderForFilter=XZY)
    Child=(Type=Bone,Name="base_cartridges_c")
    Parents="((Item=(Type=Bone,Name=\"pelvis\"),Weight=0.0001),(Item=(Type=Bone,Name=\"spine_01\"),Weight=0.0006),(Item=(Type=Bone,Name=\"clavicle_l\"),Weight=0.0016),(Item=(Type=Bone,Name=\"deform_pec_l\"),Weight=0.0025),(Item=(Type=Bone,Name=\"spine_02\"),Weight=0.0142),(Item= ...(truncated)"
    Weight=1
   )
   > &RigUnit_ParentConstraint(
    AdvancedSettings=(RotationOrderForFilter=XZY)
    Child=(Type=Bone,Name="base_cartridges_r")
    Parents="((Item=(Type=Bone,Name=\"upperarm_twist_01_r\"),Weight=0.0001),(Item=(Type=Bone,Name=\"spine_02\"),Weight=0.0034),(Item=(Type=Bone,Name=\"spine_03\"),Weight=0.0297),(Item=(Type=Bone,Name=\"clavicle_r\"),Weight=0.0788),(Item=(Type=Bone,Name=\"spine_05\"),Weight=0.084 ...(truncated)"
    Weight=1
   )
   > RigVMFunctionReturnNode()
  ]}
 New Function_ContainedGraph {nodes [
   RigVMFunctionEntryNode()
   > DeltaRotation v4 = RigVMFunctionReferenceNode(
    Rotation Order=ZYX
    child=(Type=Bone,Name="calf_l")
    outputMinMax=(X=0,Y=1)
    parent=(Type=Null,Name="calf_l_null")
    premult=1
    rotationMinMax=(X=100,Y=150)
    triggerAxis=Z
    twistAxis=(X=1,Y=0,Z=0)
   )
   > &RigUnit_ModifyTransforms(
    ItemToModify="((Item=(Type=Bone,Name=\"calf_twist_00_l\"),Transform=(Rotation=(X=0,Y=0,Z=-0.2588,W=0.9659),Translation=(X=0,Y=0,Z=0),Scale3D=(X=1,Y=1,Z=1))),(Item=(Type=Bone,Name=\"deform_knee_out_l\"),Transform=(Rotation=(X=0,Y=0,Z=0.342,W=0.9397),Translation=(X=-2,Y=0,Z=0 ...(truncated)"
    WeightMinimum=0
    Weight=@"DeltaRotation v4".outputWeight
   )
   > DeltaRotation v4_1 = RigVMFunctionReferenceNode(
    Rotation Order=ZYX
    child=(Type=Bone,Name="calf_r")
    outputMinMax=(X=0,Y=1)
    parent=(Type=Null,Name="calf_r_null")
    premult=1
    rotationMinMax=(X=100,Y=150)
    triggerAxis=Z
    twistAxis=(X=1,Y=0,Z=0)
   )
   > &RigUnit_ModifyTransforms(
    ItemToModify="((Item=(Type=Bone,Name=\"calf_twist_00_r\"),Transform=(Rotation=(X=0,Y=0,Z=-0.2588,W=0.9659),Translation=(X=0,Y=0,Z=0),Scale3D=(X=1,Y=1,Z=1))),(Item=(Type=Bone,Name=\"deform_knee_out_r\"),Transform=(Rotation=(X=0,Y=0,Z=0.342,W=0.9397),Translation=(X=2,Y=0,Z=0) ...(truncated)"
    WeightMinimum=0
    Weight=@"DeltaRotation v4_1".outputWeight
   )
   > RigVMFunctionReturnNode()
  ]}
 RigVMFunctionLibrary {nodes [
   DeltaRotation v4 = RigVMCollapseNode(
    ExecuteContext=()
    Rotation Order=ZYX
    child=(Type=Bone,Name="None")
    outputMinMax=(X=0,Y=1)
    parent=(Type=Bone,Name="None")
    premult=1
    rotationMinMax=(X=0,Y=0)
    twistAxis=(X=1,Y=0,Z=0)
   )
   belts = RigVMCollapseNode(ExecuteContext=())
   customKneeSetup = RigVMCollapseNode(ExecuteContext=())
   jacket = RigVMCollapseNode(ExecuteContext=())
   misc = RigVMCollapseNode(ExecuteContext=())
   twistFix = RigVMCollapseNode(ExecuteContext=())
  ]}
 RigVMFunction_Sequence_1_ContainedGraph {nodes [
   n1 = RigVMFunctionEntryNode()
   RigVMFunction_Sequence = &RigVMFunction_Sequence(ExecuteContext=@n1.ExecuteContext)
   n2 = &RigVMFunction_Sequence(ExecuteContext=@RigVMFunction_Sequence.B)
   RigVMFunctionReturnNode(A=@RigVMFunction_Sequence.A, B=@n2.A, C=@n2.B)
  ]}
 RigVMModel {nodes [
   RigUnit_BeginExecution = &RigUnit_BeginExecution()
   twistFix = RigVMFunctionReferenceNode(ExecuteContext=@RigUnit_BeginExecution.ExecutePin)
   > FN_DeformRig_FNC = RigVMFunctionReferenceNode(
    DeformRig_ Curve="(deformRig_l_upperarm_up_alpha_72_#1=False,deformRig_r_upperarm_up_alpha_73_#2=False,deformRig_l_upperarm_fr_alpha_74_#3=False,deformRig_r_upperarm_fr_alpha_75_CB51B8 ...(truncated)"
    DeformRig_Bool="(upperarm_1_#4=True,elbow_3_#5=True,wrist_7_#6=True,groin_8_#7=true,glute_17_#8=True,knee_9_40C408324CA261 ...(truncated)"
    DeformRig_Debug=(upperarm_debug_18_#4=False,wrist_debug_20_#6=False,lat_debug_26_#9=False,groinGlute_debug_27_#7=False)
    DeformRig_Value="(upperarm_up_18_#10=(Rotation=(X=0,Y=0,Z=0,W=1),Translation=(X=0,Y=0,Z=8),Scale3D=(X=0.7,Y=1,Z=1)),upperarm_fwd_fr_19_#11=(Rotation=(X=0,Y=0,Z=0,W=1),Translation=(X=0,Y=10,Z=0),Scale3D=(X=0.5,Y=1,Z= ...(truncated)"
   )
   > customKneeSetup = RigVMFunctionReferenceNode()
   > jacket = RigVMFunctionReferenceNode()
   > belts = RigVMFunctionReferenceNode()
   > misc = RigVMFunctionReferenceNode()
   setFKCtrl = RigVMFunctionReferenceNode(ExecuteContext=(), setFK_Ctrl=false)
  ]}
 RigVMModel Rig {nodes [
   n1 = &RigUnit_PrepareForExecution()
   &RigUnit_HierarchyImportFromSkeleton(NameSpace=None, bIncludeCurves=true, bIncludeMeshSockets=False, bIncludeVirtualBones=True, ExecutePin=@n1.ExecutePin)
   RigUnit_ItemArray = &RigUnit_ItemArray(Items="((Type=Bone,Name=\"thigh_l\"),(Type=Bone,Name=\"calf_l\"),(Type=Bone,Name=\"foot_l\"),(Type=Bone,Name=\"thigh_r\"),(Type=Bone,Name=\"calf_r\"),(Type=Bone,Name=\"foot_r\"),(Type=Bone,Name=\"clavicle_l\"),(Type=Bone,Name=\"upperarm_l\"),(Type=Bone,Name=\"hand_l\"),(Type=Bone, ...(truncated)")
   > addNullObject = RigVMFunctionReferenceNode(
    Options_Null="(pelvis_1_#12=False,spine_3_#13=False,clavicle_5_#14=False,arm_7_#15=False,leg_9_#16=False,head_11_46DC15EF465C ...(truncated)"
    Custom_Joints=@RigUnit_ItemArray.Items
   )
   addFKCtrl = RigVMFunctionReferenceNode(
    Custom_FKCtrl=()
    ExecuteContext=()
    Options_FKCtrl="(pelvis_1_#12=true,spine_3_#13=true,clavicle_5_#14=true,arm_7_#15=true,leg_9_#16=true,head_11_46DC15EF465CD2979 ...(truncated)"
    Shape_Settings="(pelvis_2_#17=(bVisible=True,Name=\"Hexagon_Thin\",Color=(R=0.0117,G=0.8281,B=1,A=1),Transform=(Rotation=(X=0,Y=-0.7071,Z=0,W=0.7071),Translation=(X=2,Y=0,Z=0),Scale3D=(X=5,Y=5,Z=7))),spine_4_#18=(bVi ...(truncated)"
   )
  ]}
]
variables [
 thigh_l_n20_n110 {type=double, default=0}
 thigh_r_n20_n110 {type=double, default=0}
]
