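# Physics asset for the M_MED_BillyGold male player cosmetic body, set up for secondary motion of the jacket hem.
# Seven box bodies on dyn_hemJacket bones are constrained with narrow swing and zero twist limits to tiny base_hemJacket anchor capsules, which link to spine_02.
# Capsules on pelvis, lower spine, thighs and selected kin_belt bones approximate body and belt shapes, with extra capsules on the left thigh.
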PhysicsAsset /BRCosmetics/Characters/Player/Male/Medium/Bodies/M_MED_BillyGold/Meshes/M_MED_BillyGold_Physics
bodies {
 base_hemJacket_1_c [{shapeType=Capsule, length=1, radius=0.5, shapeName=base_hemJacket_1_c_capsule}]
 base_hemJacket_1_l [{shapeType=Capsule, length=1, radius=0.5, shapeName=base_hemJacket_1_l_capsule}]
 base_hemJacket_1_r [{shapeType=Capsule, length=1, radius=0.5, shapeName=base_hemJacket_1_r_capsule}]
 base_hemJacket_2_l [{shapeType=Capsule, length=1, radius=0.5, shapeName=base_hemJacket_2_l_capsule}]
 base_hemJacket_2_r [{shapeType=Capsule, length=1, radius=0.5, shapeName=base_hemJacket_2_r_capsule}]
 base_hemJacket_3_l [{shapeType=Capsule, length=1, radius=0.5, shapeName=base_hemJacket_3_l_capsule}]
 base_hemJacket_3_r [{shapeType=Capsule, length=1, radius=0.5, shapeName=base_hemJacket_3_r_capsule}]
 dyn_hemJacket_1_c [{
   shapeType=Box
   extentX=9.8336
   extentY=2.0335
   extentZ=12.22
   shapeName=dyn_hemJacket_1_l_box
   center=<5.1818 0.105 0>
   rotation=<rot 0 1.1605 0>
  }]
 dyn_hemJacket_1_l [{
   shapeType=Box
   extentX=12.431
   extentY=2.0335
   extentZ=7.5
   shapeName=dyn_hemJacket_1_l_box
   center=<6.8768 0.7199 -3.2859>
   rotation=<rot -3.1284 -0.0887 -15.4385>
  }]
 dyn_hemJacket_1_r [{
   shapeType=Box
   extentX=12.431
   extentY=2.0335
   extentZ=7.5
   shapeName=dyn_hemJacket_1_l_box
   center=<7.2842 0.2486 3.2785>
   rotation=<rot 2.8 1.1 15.5>
  }]
 dyn_hemJacket_2_l [{
   shapeType=Box
   extentX=11.7732
   extentY=2.0335
   extentZ=7.5
   shapeName=dyn_hemJacket_1_l_box
   center=<6.1644 -0.0306 -1.3066>
   rotation=<rot -4.4861 1.1605 0>
  }]
 dyn_hemJacket_2_r [{
   shapeType=Box
   extentX=11.7732
   extentY=2.0335
   extentZ=7.5
   shapeName=dyn_hemJacket_1_l_box
   center=<6.1644 0.0306 1.3066>
   rotation=<rot 4.4861 -1.1605 0>
  }]
 dyn_hemJacket_3_l [{
   shapeType=Box
   extentX=10.562
   extentY=2.0335
   extentZ=7.5
   shapeName=dyn_hemJacket_1_l_box
   center=<5.456 -0.045 -0.095>
   rotation=<rot -4.4253 1.2049 -9.0685>
  }]
 dyn_hemJacket_3_r [{
   shapeType=Box
   extentX=10.562
   extentY=2.0335
   extentZ=7.5
   shapeName=dyn_hemJacket_1_l_box
   center=<5.456 -0.045 -0.095>
   rotation=<rot 4.4253 1.2049 9.0685>
  }]
 kin_belt_A_1 [{
   shapeType=Capsule
   length=4.405
   radius=3.4755
   shapeName=kin_belt_A_1_capsule
   center=<-0.2366 1.9169 -0.6508>
   rotation=<rot -8.6231 3.705 -84.2301>
  }]
 kin_belt_A_2 [{
   shapeType=Capsule
   length=5.6902
   radius=2.355
   shapeName=kin_belt_A_1_capsule
   center=<1.0139 -0.9717 -1.5703>
   rotation=<rot 9.2769 -3.0571 -20.5284>
  }]
 kin_belt_A_8 [{
   shapeType=Capsule
   length=5.5259
   radius=2.0627
   shapeName=kin_belt_A_8_capsule
   center=<-0.808 2.7157 1.5675>
   rotation=<rot 37.0476 -92.681 -20.736>
  }]
 kin_belt_B_1 [{
   shapeType=Capsule
   length=4.405
   radius=2.672
   shapeName=kin_belt_A_1_capsule
   center=<0.0368 0.7239 0.6402>
   rotation=<rot -7.9075 -2.8956 -83.2802>
  }]
 kin_belt_B_2 [{
   shapeType=Capsule
   length=5.5564
   radius=2.1718
   shapeName=kin_belt_B_2_capsule
   center=<1.1006 -1.7412 1.3526>
   rotation=<rot 12.9042 166.3179 -35.8037>
  }]
 kin_belt_B_7 [{
   shapeType=Capsule
   length=6.6828
   radius=2.3517
   shapeName=kin_belt_B_7_capsule
   center=<-0.0014 1.3355 1.1343>
   rotation=<rot 0 99.023 -21.6706>
  }]
 kin_belt_B_8 [{
   shapeType=Capsule
   length=6.8053
   radius=2.8636
   shapeName=kin_belt_B_8_capsule
   center=<-0.9582 1.5476 0.572>
   rotation=<rot -33.5534 -91.157 13.0605>
  }]
 pelvis [{shapeType=Capsule, length=10.7633, radius=9.6019, shapeName=pelvis_capsule, center=<-2.1337 -1.1654 0>, rotation=<rot 1.4979 -90 -0.6558>}]
 spine_01 [{shapeType=Capsule, length=7.1416, radius=9.4773, shapeName=spine_01_capsule, center=<-0.6637 0 0>, rotation=<rot 0 -90 0>}]
 spine_02 [{shapeType=Capsule, length=7.1565, radius=9.4475, shapeName=spine_02_capsule, center=<0 0.4299 -0.0009>, rotation=<rot 0.1248 -90 -0.0233>}]
 thigh_l [
  {
   shapeType=Capsule
   length=33.4374
   radius=8.7948
   shapeName=thigh_l_capsule
   center=<-16.8603 -0.7226 1.3397>
   rotation=<rot 87.6287 179.9998 179.9999>
  }
  {
   shapeType=Capsule
   length=11.4485
   radius=3.9364
   shapeName=thigh_l_capsule
   center=<-18.7749 2.6956 -5.58>
   rotation=<rot 88.5864 0.0003 0.0004>
  }
  {
   shapeType=Capsule
   length=13.8629
   radius=3.9364
   shapeName=thigh_l_capsule
   center=<-18.233 -2.1515 -5.7701>
   rotation=<rot 88.5864 0.0003 0.0004>
  }
 ]
 thigh_r [{
   shapeType=Capsule
   length=33.4374
   radius=8.7948
   shapeName=thigh_r_capsule
   center=<16.8603 0.7226 -1.3397>
   rotation=<rot 87.6287 179.9998 -0.0001>
  }]
}
constraints [
 {
  bone1Name=spine_02
  bone2Name=spine_01
  swing1LimitDegrees=45
  swing1Motion=Free
  swing2LimitDegrees=45
  swing2Motion=Free
  twistLimitDegrees=45
  twistMotion=Free
 }
 {
  bone1Name=spine_01
  bone2Name=pelvis
  swing1LimitDegrees=45
  swing1Motion=Free
  swing2LimitDegrees=45
  swing2Motion=Free
  twistLimitDegrees=45
  twistMotion=Free
 }
 {
  bone1Name=thigh_l
  bone2Name=pelvis
  swing1LimitDegrees=45
  swing1Motion=Free
  swing2LimitDegrees=45
  swing2Motion=Free
  twistLimitDegrees=45
  twistMotion=Free
 }
 {
  bone1Name=thigh_r
  bone2Name=pelvis
  swing1LimitDegrees=45
  swing1Motion=Free
  swing2LimitDegrees=45
  swing2Motion=Free
  twistLimitDegrees=45
  twistMotion=Free
 }
 {bone1Name=dyn_hemJacket_3_r, bone2Name=base_hemJacket_3_r, swing1LimitDegrees=20, swing2LimitDegrees=20, twistLimitDegrees=0}
 {
  bone1Name=base_hemJacket_1_c
  bone2Name=spine_02
  swing1LimitDegrees=45
  swing1Motion=Free
  swing2LimitDegrees=45
  swing2Motion=Free
  twistLimitDegrees=45
  twistMotion=Free
 }
 {bone1Name=dyn_hemJacket_1_c, bone2Name=base_hemJacket_1_c, swing1LimitDegrees=20, swing2LimitDegrees=20, twistLimitDegrees=0}
 {
  bone1Name=base_hemJacket_1_l
  bone2Name=spine_02
  swing1LimitDegrees=45
  swing1Motion=Free
  swing2LimitDegrees=45
  swing2Motion=Free
  twistLimitDegrees=45
  twistMotion=Free
 }
 {bone1Name=dyn_hemJacket_1_l, bone2Name=base_hemJacket_1_l, swing1LimitDegrees=20, swing2LimitDegrees=20, twistLimitDegrees=0}
 {
  bone1Name=base_hemJacket_1_r
  bone2Name=spine_02
  swing1LimitDegrees=45
  swing1Motion=Free
  swing2LimitDegrees=45
  swing2Motion=Free
  twistLimitDegrees=45
  twistMotion=Free
 }
 {bone1Name=dyn_hemJacket_1_r, bone2Name=base_hemJacket_1_r, swing1LimitDegrees=20, swing2LimitDegrees=20, twistLimitDegrees=0}
 {
  bone1Name=base_hemJacket_2_l
  bone2Name=spine_02
  swing1LimitDegrees=45
  swing1Motion=Free
  swing2LimitDegrees=45
  swing2Motion=Free
  twistLimitDegrees=45
  twistMotion=Free
 }
 {bone1Name=dyn_hemJacket_2_l, bone2Name=base_hemJacket_2_l, swing1LimitDegrees=20, swing2LimitDegrees=20, twistLimitDegrees=0}
 {
  bone1Name=base_hemJacket_2_r
  bone2Name=spine_02
  swing1LimitDegrees=45
  swing1Motion=Free
  swing2LimitDegrees=45
  swing2Motion=Free
  twistLimitDegrees=45
  twistMotion=Free
 }
 {bone1Name=dyn_hemJacket_2_r, bone2Name=base_hemJacket_2_r, swing1LimitDegrees=20, swing2LimitDegrees=20, twistLimitDegrees=0}
 {
  bone1Name=base_hemJacket_3_l
  bone2Name=spine_02
  swing1LimitDegrees=45
  swing1Motion=Free
  swing2LimitDegrees=45
  swing2Motion=Free
  twistLimitDegrees=45
  twistMotion=Free
 }
 {bone1Name=dyn_hemJacket_3_l, bone2Name=base_hemJacket_3_l, swing1LimitDegrees=20, swing2LimitDegrees=20, twistLimitDegrees=0}
 {
  bone1Name=base_hemJacket_3_r
  bone2Name=spine_02
  swing1LimitDegrees=45
  swing1Motion=Free
  swing2LimitDegrees=45
  swing2Motion=Free
  twistLimitDegrees=45
  twistMotion=Free
 }
 {
  bone1Name=kin_belt_A_1
  bone2Name=pelvis
  swing1LimitDegrees=45
  swing1Motion=Free
  swing2LimitDegrees=45
  swing2Motion=Free
  twistLimitDegrees=45
  twistMotion=Free
 }
 {
  bone1Name=kin_belt_A_2
  bone2Name=pelvis
  swing1LimitDegrees=45
  swing1Motion=Free
  swing2LimitDegrees=45
  swing2Motion=Free
  twistLimitDegrees=45
  twistMotion=Free
 }
 {
  bone1Name=kin_belt_A_8
  bone2Name=pelvis
  swing1LimitDegrees=45
  swing1Motion=Free
  swing2LimitDegrees=45
  swing2Motion=Free
  twistLimitDegrees=45
  twistMotion=Free
 }
 {
  bone1Name=kin_belt_B_1
  bone2Name=pelvis
  swing1LimitDegrees=45
  swing1Motion=Free
  swing2LimitDegrees=45
  swing2Motion=Free
  twistLimitDegrees=45
  twistMotion=Free
 }
 {
  bone1Name=kin_belt_B_2
  bone2Name=pelvis
  swing1LimitDegrees=45
  swing1Motion=Free
  swing2LimitDegrees=45
  swing2Motion=Free
  twistLimitDegrees=45
  twistMotion=Free
 }
 {
  bone1Name=kin_belt_B_8
  bone2Name=pelvis
  swing1LimitDegrees=45
  swing1Motion=Free
  swing2LimitDegrees=45
  swing2Motion=Free
  twistLimitDegrees=45
  twistMotion=Free
 }
 {
  bone1Name=kin_belt_B_7
  bone2Name=pelvis
  swing1LimitDegrees=45
  swing1Motion=Free
  swing2LimitDegrees=45
  swing2Motion=Free
  twistLimitDegrees=45
  twistMotion=Free
 }
]
