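# Physics asset for the face-accessory mesh part of the BillyGold medium male player body, covering its hood strings and three cartridges with capsule bodies.
# Each hood string is a three-segment chain from a small base capsule, with tight swing limits and zero twist limit between links.
# String bases and clavicles attach to spine_05, while spine_05 and the cartridges attach to spine_04, all with free swing and twist.
# Larger spine and clavicle capsules approximate the upper torso.
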PhysicsAsset /BRCosmetics/Characters/Player/Male/Medium/Bodies/M_MED_BillyGold/Meshes/Parts/M_MED_BillyGold_FaceAcc_Physics
bodies {
 base_hoodStrings_1_l [{shapeType=Capsule, length=1, radius=0.5, shapeName=base_hoodStrings_1_l_capsule}]
 base_hoodStrings_1_r [{shapeType=Capsule, length=1, radius=0.5, shapeName=base_hoodStrings_1_r_capsule}]
 clavicle_l [{
   shapeType=Capsule
   length=7.4132
   radius=9.1595
   shapeName=clavicle_l_capsule
   center=<8.3047 -0.5688 -1.0662>
   rotation=<rot -76.542 179.9996 169.5884>
  }]
 clavicle_r [{
   shapeType=Capsule
   length=7.4132
   radius=9.1595
   shapeName=clavicle_r_capsule
   center=<-8.3047 0.5688 1.0662>
   rotation=<rot -76.542 179.9996 -10.4116>
  }]
 dyn_cartridges_c [{shapeType=Capsule, length=6.9101, radius=2.6939, shapeName=dyn_cartridges_l_capsule, center=<3.4559 0.8306 0>, rotation=<rot 0 -90 90>}]
 dyn_cartridges_l [{shapeType=Capsule, length=6.9101, radius=2.6939, shapeName=dyn_cartridges_l_capsule, center=<3.4559 0.8306 0>, rotation=<rot 0 -90 90>}]
 dyn_cartridges_r [{shapeType=Capsule, length=6.9101, radius=2.6939, shapeName=dyn_cartridges_l_capsule, center=<3.4559 0.8306 0>, rotation=<rot 0 -90 90>}]
 dyn_hoodStrings_1_l [{
   shapeType=Capsule
   length=3.632
   radius=0.7066
   shapeName=dyn_hoodStrings_1_l_capsule
   center=<1.8177 0.0368 -0.0331>
   rotation=<rot 0 -88.9315 -87.998>
  }]
 dyn_hoodStrings_1_r [{
   shapeType=Capsule
   length=3.6223
   radius=0.7098
   shapeName=dyn_hoodStrings_1_r_capsule
   center=<1.8057 0.0456 0.0294>
   rotation=<rot 0 91.2441 -88.0397>
  }]
 dyn_hoodStrings_2_l [{
   shapeType=Capsule
   length=3.7056
   radius=0.79
   shapeName=dyn_hoodStrings_2_l_capsule
   center=<2.0035 -0.0452 -0.0163>
   rotation=<rot 0 89.7143 -89.1237>
  }]
 dyn_hoodStrings_2_r [{
   shapeType=Capsule
   length=3.2595
   radius=0.7998
   shapeName=dyn_hoodStrings_2_r_capsule
   center=<1.9192 -0.1047 0.0156>
   rotation=<rot 0 -93.1791 -89.4521>
  }]
 dyn_hoodStrings_3_l [{
   shapeType=Capsule
   length=3.7607
   radius=0.8384
   shapeName=dyn_hoodStrings_3_l_capsule
   center=<3.1973 -0.0497 -0.01>
   rotation=<rot 0 91.2204 -89.9817>
  }]
 dyn_hoodStrings_3_r [{
   shapeType=Capsule
   length=3.6902
   radius=0.8502
   shapeName=dyn_hoodStrings_3_r_capsule
   center=<2.8358 -0.1295 -0.0074>
   rotation=<rot 0 -88.8464 -89.2862>
  }]
 spine_04 [
  {
   shapeType=Capsule
   length=11.7978
   radius=11.2847
   shapeName=spine_04_capsule
   center=<5.1771 0.4848 0.0014>
   rotation=<rot -0.8513 -90 0.1975>
  }
  {
   shapeType=Capsule
   length=15.8833
   radius=7.9377
   shapeName=spine_04_capsule
   center=<1.4035 3.7996 -0.0094>
   rotation=<rot -0.8513 -90 0.1975>
  }
 ]
 spine_05 [{
   shapeType=Capsule
   length=12.601
   radius=11.2847
   shapeName=spine_05_capsule
   center=<-1.8102 -0.0913 -0.1518>
   rotation=<rot -0.8513 -90 0.1975>
  }]
}
constraints [
 {bone1Name=dyn_hoodStrings_1_l, bone2Name=base_hoodStrings_1_l, swing1LimitDegrees=20, swing2LimitDegrees=20, twistLimitDegrees=0}
 {bone1Name=dyn_hoodStrings_2_l, bone2Name=dyn_hoodStrings_1_l, swing1LimitDegrees=20, swing2LimitDegrees=20, twistLimitDegrees=0}
 {bone1Name=dyn_hoodStrings_3_l, bone2Name=dyn_hoodStrings_2_l, swing1LimitDegrees=20, swing2LimitDegrees=20, twistLimitDegrees=0}
 {bone1Name=dyn_hoodStrings_1_r, bone2Name=base_hoodStrings_1_r, swing1LimitDegrees=20, swing2LimitDegrees=20, twistLimitDegrees=0}
 {bone1Name=dyn_hoodStrings_2_r, bone2Name=dyn_hoodStrings_1_r, swing1LimitDegrees=20, swing2LimitDegrees=20, twistLimitDegrees=0}
 {bone1Name=dyn_hoodStrings_3_r, bone2Name=dyn_hoodStrings_2_r, swing1LimitDegrees=20, swing2LimitDegrees=20, twistLimitDegrees=0}
 {
  bone1Name=base_hoodStrings_1_l
  bone2Name=spine_05
  swing1LimitDegrees=45
  swing1Motion=Free
  swing2LimitDegrees=45
  swing2Motion=Free
  twistLimitDegrees=45
  twistMotion=Free
 }
 {
  bone1Name=base_hoodStrings_1_r
  bone2Name=spine_05
  swing1LimitDegrees=45
  swing1Motion=Free
  swing2LimitDegrees=45
  swing2Motion=Free
  twistLimitDegrees=45
  twistMotion=Free
 }
 {
  bone1Name=clavicle_l
  bone2Name=spine_05
  swing1LimitDegrees=45
  swing1Motion=Free
  swing2LimitDegrees=45
  swing2Motion=Free
  twistLimitDegrees=45
  twistMotion=Free
 }
 {
  bone1Name=clavicle_r
  bone2Name=spine_05
  swing1LimitDegrees=45
  swing1Motion=Free
  swing2LimitDegrees=45
  swing2Motion=Free
  twistLimitDegrees=45
  twistMotion=Free
 }
 {
  bone1Name=spine_05
  bone2Name=spine_04
  swing1LimitDegrees=45
  swing1Motion=Free
  swing2LimitDegrees=45
  swing2Motion=Free
  twistLimitDegrees=45
  twistMotion=Free
 }
 {
  bone1Name=dyn_cartridges_r
  bone2Name=spine_04
  swing1LimitDegrees=45
  swing1Motion=Free
  swing2LimitDegrees=45
  swing2Motion=Free
  twistLimitDegrees=45
  twistMotion=Free
 }
 {
  bone1Name=dyn_cartridges_l
  bone2Name=spine_04
  swing1LimitDegrees=45
  swing1Motion=Free
  swing2LimitDegrees=45
  swing2Motion=Free
  twistLimitDegrees=45
  twistMotion=Free
 }
 {
  bone1Name=dyn_cartridges_c
  bone2Name=spine_04
  swing1LimitDegrees=45
  swing1Motion=Free
  swing2LimitDegrees=45
  swing2Motion=Free
  twistLimitDegrees=45
  twistMotion=Free
 }
]
