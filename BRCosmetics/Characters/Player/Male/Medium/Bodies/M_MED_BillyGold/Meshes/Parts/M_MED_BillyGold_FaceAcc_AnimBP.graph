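# Post-style AnimBlueprint for the BillyGold character's face accessory mesh, layering body-mesh arm and shoulder-pad bones over the attached parent's pose.
# Pec bones are overridden before a Control Rig pass, then cartridge bones before hood AnimDynamics and a spine_05-space RigidBody simulation.
# Each update computes a gravity override and, if the body has upperarm_r, caches parent-space transforms of those pec and cartridge bones.
# Customization load supplies the body mesh.
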AnimBlueprint /BRCosmetics/Characters/Player/Male/Medium/Bodies/M_MED_BillyGold/Meshes/Parts/M_MED_BillyGold_FaceAcc_AnimBP
target_skeleton=&/M_MED_BillyGold_FaceAcc_Skeleton
graphs [
 AnimGraph {nodes [
   Output Pose(Result=Use cached pose 'dynamicsCP'()) {layerGroup=None}
   n1 = Layered blend per bone(
    BasePose=Copy Pose From Mesh() {bUseAttachedParent=true}
    BlendPoses_0=Copy Pose From Mesh(SourceMeshComponent=Get BodyMesh() -> (BodyMesh:Skeletal Mesh Component Object Reference))
   ) {layerSetup [{branchFilters [
       {boneName=clavicle_l}
       {boneName=clavicle_r}
       {boneName=base_armourPadArm_1}
       {boneName=dyn_armourPadArm_1}
       {boneName=base_armourPadArm_bend}
       {boneName=dyn_armourPadArm_2}
       {boneName=dyn_armourPadArm_bend}
       {boneName=dyn_armourPadArm_bend_end}
       {boneName=dyn_armourPadArm_2_end}
       {boneName=upperarm_r}
      ]}]}
   Save cached pose 'InputCP'(Pose=@n1) {cacheName=InputCP}
   n2 = "Transform (Modify) Bone - Bone: base_cartridges_l"!(
    ComponentPose=Local To Component(LocalPose=Use cached pose 'controlRigPecBonesCP'())
    bAlphaBoolEnabled:bool=True
    Translation="0,0,0"
    Rotation="0,0,0"
   ) -> (Pose) {
    alphaInputType=Bool
    rotationMode=BMM_Replace
    rotationSpace=BCS_ParentBoneSpace
    translationMode=BMM_Replace
    translationSpace=BCS_ParentBoneSpace
    alphaBoolBlend {blendOption=Linear}
    boneToModify {boneName=base_cartridges_l}
   }
   n3 = "Transform (Modify) Bone - Bone: base_cartridges_c"!(ComponentPose=@n2, bAlphaBoolEnabled:bool=True, Translation="0,0,0", Rotation="0,0,0") -> (Pose) {
    alphaInputType=Bool
    rotationMode=BMM_Replace
    rotationSpace=BCS_ParentBoneSpace
    translationMode=BMM_Replace
    translationSpace=BCS_ParentBoneSpace
    alphaBoolBlend {blendOption=Linear}
    boneToModify {boneName=base_cartridges_c}
   }
   n4 = "Transform (Modify) Bone - Bone: base_cartridges_r"!(ComponentPose=@n3, bAlphaBoolEnabled:bool=True, Translation="0,0,0", Rotation="0,0,0") -> (Pose) {
    alphaInputType=Bool
    rotationMode=BMM_Replace
    rotationSpace=BCS_ParentBoneSpace
    translationMode=BMM_Replace
    translationSpace=BCS_ParentBoneSpace
    alphaBoolBlend {blendOption=Linear}
    boneToModify {boneName=base_cartridges_r}
   }
   n5 = "Transform (Modify) Bone - Bone: dyn_cartridges_l"!(ComponentPose=@n4, bAlphaBoolEnabled:bool=True, Translation="0,0,0", Rotation="0,0,0") -> (Pose) {
    alphaInputType=Bool
    rotationMode=BMM_Replace
    rotationSpace=BCS_ParentBoneSpace
    translationMode=BMM_Replace
    translationSpace=BCS_ParentBoneSpace
    alphaBoolBlend {blendOption=Linear}
    boneToModify {boneName=dyn_cartridges_l}
   }
   n6 = "Transform (Modify) Bone - Bone: dyn_cartridges_c"!(ComponentPose=@n5, bAlphaBoolEnabled:bool=True, Translation="0,0,0", Rotation="0,0,0") -> (Pose) {
    alphaInputType=Bool
    rotationMode=BMM_Replace
    rotationSpace=BCS_ParentBoneSpace
    translationMode=BMM_Replace
    translationSpace=BCS_ParentBoneSpace
    alphaBoolBlend {blendOption=Linear}
    boneToModify {boneName=dyn_cartridges_c}
   }
   n7 = "Transform (Modify) Bone - Bone: dyn_cartridges_r"!(ComponentPose=@n6, bAlphaBoolEnabled:bool=True, Translation="0,0,0", Rotation="0,0,0") -> (Pose) {
    alphaInputType=Bool
    rotationMode=BMM_Replace
    rotationSpace=BCS_ParentBoneSpace
    translationMode=BMM_Replace
    translationSpace=BCS_ParentBoneSpace
    alphaBoolBlend {blendOption=Linear}
    boneToModify {boneName=dyn_cartridges_r}
   }
   n8 = Anim Dynamics - Bone: dyn_hoodTop_c!(ComponentPose=@n7, Alpha=1, GravityOverride:Vector="0,0,-980", ExternalForce="0,0,0") -> (Pose) {
    angularDampingOverride=1
    angularSpringConstant=1000
    bAngularSpring=true
    bOverrideAngularDamping=true
    bUseGravityOverride=true
    gravityScale=0.35
    simulationSpace=RootRelative
    alphaBoolBlend {blendOption=Linear}
    boundBone {boneName=dyn_hoodTop_c}
    componentAppliedLinearAccClamp=<4000 4000 4000>
    componentLinearAccScale=<0.4 0.4 0.6>
    componentLinearVelScale=<0.4 0.4 0.6>
    gravityOverride {z=-980}
    physicsBodyDefinitions [{
      boundBone {boneName=dyn_hoodTop_c}
      boxExtents=<20 20 20>
      constraintSetup {angularLimitsMax=<0 10 15>, angularLimitsMin=<0 -10 -5>, angularTarget=<1 0 0>}
      localJointOffset=<1 0 0>
     }]
   }
   n9 = RigidBody!(ComponentPose=@n8, Alpha, OverrideWorldGravity:Vector="0,0,0", ExternalForce) -> (Pose) {
    bOverrideWorldGravity=true
    overridePhysicsAsset=&/M_MED_BillyGold_FaceAcc_Physics
    simulationSpace=BaseBoneSpace
    alphaBoolBlend {blendOption=Linear}
    alphaScaleBiasClamp {interpSpeedDecreasing=0.1, interpSpeedIncreasing=0.1}
    baseBoneRef {boneName=spine_05}
    componentAppliedLinearAccClamp=<3000 3000 3000>
    overrideWorldGravity {z=-980}
    simSpaceSettings {
     maxAngularAcceleration=1
     maxAngularVelocity=3
     maxLinearAcceleration=1000
     maxLinearVelocity=500
     velocityScaleZ=0.5
     worldAlpha=1
     externalLinearDragV=<0.2 0.2>
    }
   }
   Save cached pose 'dynamicsCP'(Pose=Component To Local(ComponentPose=@n9)) {cacheName=dynamicsCP}
   n10 = "Transform (Modify) Bone - Bone: deform_pec_aim_base_l"!(ComponentPose=Local To Component(LocalPose=Use cached pose 'InputCP'()), bAlphaBoolEnabled:bool=True, Translation="0,0,0", Rotation="0,0,0") -> (Pose) {
    alphaInputType=Bool
    rotationMode=BMM_Replace
    rotationSpace=BCS_ParentBoneSpace
    translationMode=BMM_Replace
    translationSpace=BCS_ParentBoneSpace
    alphaBoolBlend {blendOption=Linear}
    boneToModify {boneName=deform_pec_aim_base_l}
   }
   n11 = "Transform (Modify) Bone - Bone: deform_pec_aim_l"!(ComponentPose=@n10, bAlphaBoolEnabled:bool=True, Translation="0,0,0", Rotation="0,0,0") -> (Pose) {
    alphaInputType=Bool
    rotationMode=BMM_Replace
    rotationSpace=BCS_ParentBoneSpace
    translationMode=BMM_Replace
    translationSpace=BCS_ParentBoneSpace
    alphaBoolBlend {blendOption=Linear}
    boneToModify {boneName=deform_pec_aim_l}
   }
   n12 = "Transform (Modify) Bone - Bone: deform_pec_base_l"!(ComponentPose=@n11, bAlphaBoolEnabled:bool=True, Translation="0,0,0", Rotation="0,0,0") -> (Pose) {
    alphaInputType=Bool
    rotationMode=BMM_Replace
    rotationSpace=BCS_ParentBoneSpace
    translationMode=BMM_Replace
    translationSpace=BCS_ParentBoneSpace
    alphaBoolBlend {blendOption=Linear}
    boneToModify {boneName=deform_pec_base_l}
   }
   n13 = "Transform (Modify) Bone - Bone: deform_pec_l"!(ComponentPose=@n12, bAlphaBoolEnabled:bool=True, Translation="0,0,0", Rotation="0,0,0") -> (Pose) {
    alphaInputType=Bool
    rotationMode=BMM_Replace
    rotationSpace=BCS_ParentBoneSpace
    translationMode=BMM_Replace
    translationSpace=BCS_ParentBoneSpace
    alphaBoolBlend {blendOption=Linear}
    boneToModify {boneName=deform_pec_l}
   }
   n14 = "Transform (Modify) Bone - Bone: deform_pec_aim_base_r"!(ComponentPose=@n13, bAlphaBoolEnabled:bool=True, Translation="0,0,0", Rotation="0,0,0") -> (Pose) {
    alphaInputType=Bool
    rotationMode=BMM_Replace
    rotationSpace=BCS_ParentBoneSpace
    translationMode=BMM_Replace
    translationSpace=BCS_ParentBoneSpace
    alphaBoolBlend {blendOption=Linear}
    boneToModify {boneName=deform_pec_aim_base_r}
   }
   n15 = "Transform (Modify) Bone - Bone: deform_pec_aim_r"!(ComponentPose=@n14, bAlphaBoolEnabled:bool=True, Translation="0,0,0", Rotation="0,0,0") -> (Pose) {
    alphaInputType=Bool
    rotationMode=BMM_Replace
    rotationSpace=BCS_ParentBoneSpace
    translationMode=BMM_Replace
    translationSpace=BCS_ParentBoneSpace
    alphaBoolBlend {blendOption=Linear}
    boneToModify {boneName=deform_pec_aim_r}
   }
   n16 = "Transform (Modify) Bone - Bone: deform_pec_base_r"!(ComponentPose=@n15, bAlphaBoolEnabled:bool=True, Translation="0,0,0", Rotation="0,0,0") -> (Pose) {
    alphaInputType=Bool
    rotationMode=BMM_Replace
    rotationSpace=BCS_ParentBoneSpace
    translationMode=BMM_Replace
    translationSpace=BCS_ParentBoneSpace
    alphaBoolBlend {blendOption=Linear}
    boneToModify {boneName=deform_pec_base_r}
   }
   n17 = "Transform (Modify) Bone - Bone: deform_pec_r"!(ComponentPose=@n16, bAlphaBoolEnabled:bool=True, Translation="0,0,0", Rotation="0,0,0") -> (Pose) {
    alphaInputType=Bool
    rotationMode=BMM_Replace
    rotationSpace=BCS_ParentBoneSpace
    translationMode=BMM_Replace
    translationSpace=BCS_ParentBoneSpace
    alphaBoolBlend {blendOption=Linear}
    boneToModify {boneName=deform_pec_r}
   }
   n18 = Control Rig(Source=Component To Local(ComponentPose=@n17)) {
    alphaBoolBlend {blendOption=Linear}
    controlRigAssetReference {blueprintRigClass=&/M_MED_BillyGold_FaceAcc_CtrlRig.M_MED_BillyGold_FaceAcc_CtrlRig_C}
   }
   Save cached pose 'controlRigPecBonesCP'(Pose=@n18) {cacheName=controlRigPecBonesCP}
  ]}
 EventGraph {nodes [
   Event BlueprintUpdateAnimation()
   > n1 = Sequence()
   > n2 = Gravity Override State Function DEPRECATED(CharacterPartAnimInstance=Self-Reference(), GameTimeInSeconds=GetGameTimeInSeconds(), GravityOverrideParams=Get Gravity Override Params())
   > Set GravityOverride(GravityOverride=@n2)
   n3 = Get BodyMesh()
   Is Valid(exec=@n1.then_1, InputObject=@n3)
   n4 = GetSocketBoneName!(self:Skinned Mesh Component Object Reference=@n3, InSocketName:Name=upperarm_r) -> (ReturnValue:Name=None)
   > n5 = Set isBodyAttached(isBodyAttached:bool="Not Equal Exactly (String)"(A="To String (Name)"(InName=@n4), B=None)) -> (Output_Get:bool)
   > Branch(Condition=@n5)
   n6 = Get BodyMesh()
   n7 = BreakTransform(InTransform=GetSocketTransform(self=@n6, InSocketName=base_cartridges_l, TransformSpace=RTS_ParentBoneSpace))
   > Set base_cartridges_l_location(base_cartridges_l_location:Vector=@n7) -> (Output_Get:Vector)
   > Set base_cartridges_l_rotation(base_cartridges_l_rotation:Rotator=@n7.Rotation) -> (Output_Get:Rotator)
   n8 = BreakTransform(InTransform=GetSocketTransform(self=@n6, InSocketName=base_cartridges_c, TransformSpace=RTS_ParentBoneSpace))
   > Set base_cartridges_c_location(base_cartridges_c_location:Vector=@n8) -> (Output_Get:Vector)
   > Set base_cartridges_c_rotation(base_cartridges_c_rotation:Rotator=@n8.Rotation) -> (Output_Get:Rotator)
   n9 = BreakTransform(InTransform=GetSocketTransform(self=@n6, InSocketName=base_cartridges_r, TransformSpace=RTS_ParentBoneSpace))
   > Set base_cartridges_r_location(base_cartridges_r_location:Vector=@n9) -> (Output_Get:Vector)
   > Set base_cartridges_r_rotation(base_cartridges_r_rotation:Rotator=@n9.Rotation) -> (Output_Get:Rotator)
   n10 = BreakTransform(InTransform=GetSocketTransform(self=@n6, InSocketName=dyn_cartridges_l, TransformSpace=RTS_ParentBoneSpace))
   > Set dyn_cartridges_l_location(dyn_cartridges_l_location:Vector=@n10) -> (Output_Get:Vector)
   > Set dyn_cartridges_l_rotation(dyn_cartridges_l_rotation:Rotator=@n10.Rotation) -> (Output_Get:Rotator)
   n11 = BreakTransform(InTransform=GetSocketTransform(self=@n6, InSocketName=dyn_cartridges_c, TransformSpace=RTS_ParentBoneSpace))
   > Set dyn_cartridges_c_location(dyn_cartridges_c_location:Vector=@n11) -> (Output_Get:Vector)
   > Set dyn_cartridges_c_rotation(dyn_cartridges_c_rotation:Rotator=@n11.Rotation) -> (Output_Get:Rotator)
   n12 = BreakTransform(InTransform=GetSocketTransform(self=@n6, InSocketName=dyn_cartridges_r, TransformSpace=RTS_ParentBoneSpace))
   > Set dyn_cartridges_r_location(dyn_cartridges_r_location:Vector=@n12) -> (Output_Get:Vector)
   > Set dyn_cartridges_r_rotation(dyn_cartridges_r_rotation:Rotator=@n12.Rotation) -> (Output_Get:Rotator)
   n13 = Get BodyMesh()
   n14 = BreakTransform(InTransform=GetSocketTransform(self=@n13, InSocketName=deform_pec_aim_base_l, TransformSpace=RTS_ParentBoneSpace))
   > Set deform_pec_aim_base_l_location(deform_pec_aim_base_l_location:Vector=@n14) -> (Output_Get:Vector)
   > Set deform_pec_aim_base_l_rotation(deform_pec_aim_base_l_rotation:Rotator=@n14.Rotation) -> (Output_Get:Rotator)
   n15 = BreakTransform(InTransform=GetSocketTransform(self=@n13, InSocketName=deform_pec_aim_l, TransformSpace=RTS_ParentBoneSpace))
   > Set deform_pec_aim_l_location(deform_pec_aim_l_location:Vector=@n15) -> (Output_Get:Vector)
   > Set deform_pec_aim_l_rotation(deform_pec_aim_l_rotation:Rotator=@n15.Rotation) -> (Output_Get:Rotator)
   n16 = BreakTransform(InTransform=GetSocketTransform(self=@n13, InSocketName=deform_pec_base_l, TransformSpace=RTS_ParentBoneSpace))
   > Set deform_pec_base_l_location(deform_pec_base_l_location:Vector=@n16) -> (Output_Get:Vector)
   > Set deform_pec_base_l_rotation(deform_pec_base_l_rotation:Rotator=@n16.Rotation) -> (Output_Get:Rotator)
   n17 = BreakTransform(InTransform=GetSocketTransform(self=@n13, InSocketName=deform_pec_l, TransformSpace=RTS_ParentBoneSpace))
   > Set deform_pec_l_location(deform_pec_l_location:Vector=@n17) -> (Output_Get:Vector)
   > Set deform_pec_l_rotation(deform_pec_l_rotation:Rotator=@n17.Rotation) -> (Output_Get:Rotator)
   n18 = Get BodyMesh()
   n19 = BreakTransform(InTransform=GetSocketTransform(self=@n18, InSocketName=deform_pec_aim_base_r, TransformSpace=RTS_ParentBoneSpace))
   > Set deform_pec_aim_base_r_location(deform_pec_aim_base_r_location:Vector=@n19) -> (Output_Get:Vector)
   > Set deform_pec_aim_base_r_rotation(deform_pec_aim_base_r_rotation:Rotator=@n19.Rotation) -> (Output_Get:Rotator)
   n20 = BreakTransform(InTransform=GetSocketTransform(self=@n18, InSocketName=deform_pec_aim_r, TransformSpace=RTS_ParentBoneSpace))
   > Set deform_pec_aim_r_location(deform_pec_aim_r_location:Vector=@n20) -> (Output_Get:Vector)
   > Set deform_pec_aim_r_rotation(deform_pec_aim_r_rotation:Rotator=@n20.Rotation) -> (Output_Get:Rotator)
   n21 = BreakTransform(InTransform=GetSocketTransform(self=@n18, InSocketName=deform_pec_base_r, TransformSpace=RTS_ParentBoneSpace))
   > Set deform_pec_base_r_location(deform_pec_base_r_location:Vector=@n21) -> (Output_Get:Vector)
   > Set deform_pec_base_r_rotation(deform_pec_base_r_rotation:Rotator=@n21.Rotation) -> (Output_Get:Rotator)
   n22 = BreakTransform(InTransform=GetSocketTransform(self=@n18, InSocketName=deform_pec_r, TransformSpace=RTS_ParentBoneSpace))
   > Set deform_pec_r_location(deform_pec_r_location:Vector=@n22) -> (Output_Get:Vector)
   > Set deform_pec_r_rotation(deform_pec_r_rotation:Rotator=@n22.Rotation) -> (Output_Get:Rotator)
   n23 = Event OnCharacterCustomizationLoaded()
   > Set BodyMesh(BodyMesh=Get Array Item with EFort Custom Part Type(Array=@n23.CustomMeshParts))
  ]}
]
variables [
 base_cartridges_c_location {type=Vector}
 base_cartridges_c_rotation {type=Rotator}
 base_cartridges_l_location {type=Vector}
 base_cartridges_l_rotation {type=Rotator}
 base_cartridges_r_location {type=Vector}
 base_cartridges_r_rotation {type=Rotator}
 BodyMesh {type=SkeletalMeshComponent}
 deform_pec_aim_base_l_location {type=Vector}
 deform_pec_aim_base_l_rotation {type=Rotator}
 deform_pec_aim_base_r_location {type=Vector}
 deform_pec_aim_base_r_rotation {type=Rotator}
 deform_pec_aim_l_location {type=Vector}
 deform_pec_aim_l_rotation {type=Rotator}
 deform_pec_aim_r_location {type=Vector}
 deform_pec_aim_r_rotation {type=Rotator}
 deform_pec_base_l_location {type=Vector}
 deform_pec_base_l_rotation {type=Rotator}
 deform_pec_base_r_location {type=Vector}
 deform_pec_base_r_rotation {type=Rotator}
 deform_pec_l_location {type=Vector}
 deform_pec_l_rotation {type=Rotator}
 deform_pec_r_location {type=Vector}
 deform_pec_r_rotation {type=Rotator}
 dyn_cartridges_c_location {type=Vector}
 dyn_cartridges_c_rotation {type=Rotator}
 dyn_cartridges_l_location {type=Vector}
 dyn_cartridges_l_rotation {type=Rotator}
 dyn_cartridges_r_location {type=Vector}
 dyn_cartridges_r_rotation {type=Rotator}
 Gravity Override Params {type=GravityOverrideParamsStruct}
 GravityOverride {type=Vector}
 isBodyAttached {type=boolean}
]
